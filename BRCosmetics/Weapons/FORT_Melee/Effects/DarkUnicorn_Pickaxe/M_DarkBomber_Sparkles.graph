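# Translucent unlit sparkle particle material for the Dark Unicorn pickaxe effects, usable on Niagara and Cascade sprites.
# Particle random picks among sparkle, spark and starburst masks via 3ColorBlend, which blends two mane color parameters, one slightly negative with negative emissive allowed.
# The mobile path uses only the starburst, with the color blend reversed.
# Particle color tints emissive through MF_FX_EmissiveAdjustments; particle alpha multiplies the selected mask.
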Material /BRCosmetics/Weapons/FORT_Melee/Effects/DarkUnicorn_Pickaxe/M_DarkBomber_Sparkles
graph {
 properties {
  bAllowNegativeEmissiveColor=true
  bUsedWithNiagaraSprites=true
  bUsedWithParticleSprites=true
  blendMode=BLEND_Translucent
  pixelDepthOffsetMode=PDOM_Legacy
  shadingModel=MSM_Unlit
  usedShadingModels=Unlit
 }
 nodes [
  n1 = TextureSample() {texture=/Game/Effects/Fort_Effects/Textures/Flares/T_Starburst}
  n2 = ParticleRandom()
  n3 = MaterialFunctionCall(
   A=TextureSample() {texture=/Game/Effects/Fort_Effects/Textures/Sparkle/T_Sparkle_01}.R
   B=TextureSample() {texture=/Game/CharClasses/Outlander/Gadgets/AntiMaterialCharge/Materials/T_AntiMaterialCharge_Sparks_01}.G
   C=@n1.G
   Alpha=Ceil(None=Multiply(A=@n2) {constB=3})
  ) {materialFunction=&3ColorBlend}
  n4 = VectorParameter() {parameterName=Color_ManeBase_02, defaultValue=<rgba -0.05 -0.05 -0.05 1>}
  n5 = VectorParameter() {parameterName=Color_ManeBase_01, defaultValue=<rgba 0.114 0.0507 0.1406 1>}
  n6 = Multiply(A=@n3.Result, B=LinearInterpolate(A=@n4.RGB, B=@n5.RGB, Alpha=@n3.Result))
  n7 = Multiply(A=LinearInterpolate(A=@n5.RGB, B=@n4.RGB, Alpha=@n1.G), B=@n1.G)
  n8 = ParticleColor()
  MaterialFunctionCall(Emissive=Multiply(A=ShadingPathSwitch(Default=@n6, Mobile=@n7) {desc=Migrated from FeatureLevelSwitch}, B=@n8.RGB)) {materialFunction=&MF_FX_EmissiveAdjustments}
  Multiply(A=@n3.Result, B=@n8.A)
 ]
}
