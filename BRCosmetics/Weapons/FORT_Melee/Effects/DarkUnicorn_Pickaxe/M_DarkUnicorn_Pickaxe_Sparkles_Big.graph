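# Additive unlit particle material for the big sparkles of the Dark Unicorn pickaxe effect, usable on Cascade and Niagara sprites.
# Glyph glow color times particle color feeds an HDR emissive function with sine-pulsed intensity, alongside a four-point star texture scaled by particle alpha and a CameraOffset node.
# Separate glyph chains sample a Cube height mask via SubUV and panning, driven by dynamic parameters; the Selection blend has identical inputs, so has no effect.
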Material /BRCosmetics/Weapons/FORT_Melee/Effects/DarkUnicorn_Pickaxe/M_DarkUnicorn_Pickaxe_Sparkles_Big
graph {
 properties {
  bAllowNegativeEmissiveColor=true
  bEnableResponsiveAA=true
  bUsedWithNiagaraSprites=true
  bUsedWithParticleSprites=true
  blendMode=BLEND_Additive
  pixelDepthOffsetMode=PDOM_Legacy
  shadingModel=MSM_Unlit
  usedShadingModels=Unlit
 }
 nodes [
  n1 = VectorParameter() {group=T H E C U B E, parameterName=Glyph Glow Color, defaultValue=<rgba 0.4689 0.042 0.8958 1>}
  n2 = ParticleColor()
  n3 = Multiply(A=@n1.RGB, B=@n2.RGB)
  n4 = ScalarParameter() {defaultValue=0.8, group=T H E C U B E, parameterName=Overall Emissive Min}
  n5 = ScalarParameter() {defaultValue=1, group=T H E C U B E, parameterName=Overall Emissive Max}
  n6 = Sine(None=Multiply(A=Time(), B=ScalarParameter() {defaultValue=2.5, group=T H E C U B E, parameterName=Overall Emissive Pulse Speed}))
  MaterialFunctionCall(Emissive=MaterialFunctionCall(LDR Input=Clamp(None=@n3), HDR Tint=@n3, HDR Intensity Multiplier=LinearInterpolate(A=@n4, B=@n5, Alpha=@n6)) {materialFunction=&DeriveHDRfromLDR}.RemappedRange) {materialFunction=&MF_FX_EmissiveAdjustments}
  MaterialFunctionCall(Offset Amount=ScalarParameter() {defaultValue=25, parameterName="Offset Amount (S)"}) {materialFunction=/Engine/Functions/Engine_MaterialFunctions02/WorldPositionOffset/CameraOffset}
  n7 = Constant() {r=2}
  n8 = DynamicParameter() {paramNames [Glyph,Opacity,Pulse,Selection]}
  n9 = TextureSample(UVs=MaterialFunctionCall!(U_Images=@n7, V_Images=@n7, SubUV Texture RGB, Time Speed, Manual Time Control=StaticBool() {value=true}, Time=@n8.Glyph, UVs) {materialFunction=/Game/Effects/Fort_Effects/Materials/_Functions/MF_SubUV}.UVs) {samplerType=SAMPLERTYPE_LinearColor, texture=/Game/Athena/Prototype/Blueprints/Cube/Materials/DFHeightMask}
  n10 = Constant() {r=0.5}
  n11 = Constant() {r=1}
  n12 = Constant()
  n13 = Power(
   Base=LinearInterpolate(A=@n9.R, B=@n9.R, Alpha=If(A=@n8.Selection, B=@n10, "A > B"=@n11, "A == B"=@n12, "A < B"=@n12))
   Exp=LinearInterpolate(
    A=ScalarParameter() {defaultValue=25}
    B=ScalarParameter() {defaultValue=6}
    Alpha=DynamicParameter() {paramNames [Glyph,Opacity,Pulse,Selection]}.Pulse
   )
  )
  n14 = ScalarParameter() {defaultValue=5, parameterName=gsehgesaf}
  Multiply(A=@n1.RGB, B=Saturate(None=Multiply(A=@n13, B=@n14)))
  Multiply(A=@n2.A, B=TextureSample() {texture=/Game/Effects/Fort_Effects/Textures/Sparks/T_Sparkles_4Star_01_D}.RGB)
  n15 = ScalarParameter() {defaultValue=1, parameterName=gdfsdf}
  n16 = ScalarParameter() {parameterName=gdsgsdf}
  n17 = ScalarParameter() {defaultValue=0.0625, desc=0.125, group=T H E C U B E, parameterName=Glyph Panner Independent Scale}
  n18 = TextureCoordinate()
  n19 = ScalarParameter() {defaultValue=0.1, desc=0.0135, group=T H E C U B E, parameterName="glyph panner speed "}
  n20 = Time()
  n21 = TextureSample!(
   UVs=Panner(Coordinate=Multiply(A=@n17, B=@n18), Time=Multiply(A=@n19, B=@n20)) {speedY=1}
   Tex
   Bias=ScalarParameter() {defaultValue=2, parameterName=mip bias aagasdf}
   Apply View MipBias
  ) {
   mipValueMode=TMVM_MipBias
   samplerSource=SSM_Wrap_WorldGroupSettings
   samplerType=SAMPLERTYPE_LinearColor
   texture=&DFHeightMask
  }
  Power(
   Base=Saturate(None=LinearInterpolate(A=@n15, B=@n16, Alpha=@n21.B))
   Exp=LinearInterpolate(Alpha=DynamicParameter() {paramNames [Glyph,Opacity,Pulse,Selection]}.Opacity) {constA=10, constB=0}
  )
 ]
}
